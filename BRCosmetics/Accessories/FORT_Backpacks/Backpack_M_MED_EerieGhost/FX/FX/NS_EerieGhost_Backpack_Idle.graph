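# Idle effect for the EerieGhost backpack cosmetic, apparently a holographic logo with red glows, exposing material and texture user parameters.
# Core burst-spawns persistent mesh particles driven by system-level AutoRotate and LogoOffset values.
# PresetFX_Glow and BackingGlow sprites follow the offset and flicker via looped-age curves, with BackingGlow also animating a material parameter.
# AberrationAndGhost and RainbowLine presets are disabled; the latter switches settings on User.IsFrom3D.
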NiagaraSystem /BRCosmetics/Accessories/FORT_Backpacks/Backpack_M_MED_EerieGhost/FX/FX/NS_EerieGhost_Backpack_Idle
dependencies {
 usedDataInterfaces [&NiagaraDataInterfaceCurve]
 usedDynamicInputs [
  &RandomRangeFloat
  &MultiplyQuaternion
  &MakeQuatFromAxisAngle
  &Lerp_Vector
  &Sine
  &Add_Float
  &Lerp_Float
  &MakeVector
  &Subtract_Float
 ]
 usedModules [
  &self:SystemSpawnScript.SystemScriptSource.SystemScriptGraph.NiagaraNodeAssignment_0.SetVariables_#1
  &SystemState
  &self:SystemSpawnScript.SystemScriptSource.SystemScriptGraph.NiagaraNodeAssignment_1.SetVariables_#2
  &self:SystemSpawnScript.SystemScriptSource.SystemScriptGraph.NiagaraNodeAssignment_2.SetVariables_#3
 ]
}
emitters {
 BackingGlow {
  emitterUpdateScript [
   &EmitterState(
    Inactive Response=::NewEnumerator1
    Life Cycle Mode=::NewEnumerator1
    Loop Behavior=::NewEnumerator1
    Loop Duration Mode=::NewEnumerator1
    UseLoopDelay=false
   )
   &SpawnBurst_Instantaneous(Use Spawn Probability=false, UseLoopCountLimit=false)
  ]
  particleSpawnScript [&InitializeParticle(
    Color=<rgba 1 0 0.0667 0.5007>
    Sprite Size=<46.5235 50.0573>
    Sprite Size Mode=::NewEnumerator4
    UsePositionOffset=false
    Write Material Random=true
   )]
  particleUpdateScript [
   &ParticleState(Kill Particles When Lifetime Has Elapsed=false)
   &ScaleColor(
    Scale RGB=&VectorFromFloat(Value=&FloatFromCurve(
      FloatCurve=DataInterface() {propertyValues {Curve {keys [<at 0 1>,<at 0.1 0.625>,<at 0.2 1.125>,<at 0.4 0.375>,<at 0.5 1.375>,<at 0.6 0.75>,<at 0.7 1.625>,<at 0.8 0.625>,<at 0.9 1.25>,<at 1 0.875>]}}}
      CurveIndex=&Multiply_Float(A=$System.LoopedAge, B=0.2)
     ))
    ScaleA=true
    ScaleRGB=true
   )
   /SetVariables(Particles.Position=&Add_Vector(A=<0 3 0>, B=$System.LogoOffset)) {isSetParameters=true, module=SetVariables_#4}
   &CameraOffset(Camera Offset Amount=10)
   /SetVariables() {isSetParameters=true}
   /SetVariables(Particles.SpriteFacing=<0 -1 0>) {isSetParameters=true}
   &DynamicMaterialParameters(
    Index 0 Param 1=&FloatFromCurve(
     FloatCurve=DataInterface() {propertyValues {Curve {keys [
         <at 0 1 RCIM_Cubic>
         {arriveTangent=-0.4167, interpMode=RCIM_Cubic, leaveTangent=-0.4167, time=0.1, value=-1}
         {arriveTangent=-0.4167, interpMode=RCIM_Cubic, leaveTangent=-0.4167, time=0.3, value=0.875}
         {arriveTangent=3.125, interpMode=RCIM_Cubic, leaveTangent=3.125, time=0.4, value=-1.125}
         {arriveTangent=1.25, interpMode=RCIM_Cubic, leaveTangent=1.25, time=0.5, value=1.5}
         {arriveTangent=0.625, interpMode=RCIM_Cubic, leaveTangent=0.625, time=0.6, value=-0.875}
         {arriveTangent=0.625, interpMode=RCIM_Cubic, leaveTangent=0.625, time=0.7, value=1.625}
         {arriveTangent=-2.0833, interpMode=RCIM_Cubic, leaveTangent=-2.0833, time=0.8, value=-0.75}
         <at 1 1 RCIM_Cubic>
        ]}}}
     CurveIndex=&Multiply_Float(A=$System.LoopedAge, B=1)
    )
    Module.Param0WriteEnabled=true
    Module.Param1WriteEnabled=true
    Module.Param2WriteEnabled=true
    Module.Param3WriteEnabled=true
   )
  ]
  rendererClasses [&NiagaraSpriteRendererProperties]
  renderers [{rendererClass=&NiagaraSpriteRendererProperties}]
 }
 Core {
  emitterUpdateScript [
   &EmitterState(
    Inactive Response=::NewEnumerator1
    Life Cycle Mode=::NewEnumerator1
    Loop Behavior=::NewEnumerator1
    Loop Duration Mode=::NewEnumerator1
    UseLoopDelay=false
   )
   &SpawnBurst_Instantaneous(Use Spawn Probability=false, UseLoopCountLimit=false)
  ]
  particleSpawnScript [
   &InitializeParticle(
    Color=<rgba 250 1 1 1>
    Lifetime=5
    Mesh Scale Mode=::NewEnumerator3
    Mesh Uniform Scale=0.35
    UsePositionOffset=false
    Write Material Random=true
   )
   &DynamicMaterialParameters(
    Index 0 Param 1=$System.Random
    Index 0 Param 2=&MakeFloatFromVector(Vector=$Particles.Scale, Channel=::NewEnumerator0)
    Module.Param0WriteEnabled=true
    Module.Param1WriteEnabled=true
    Module.Param2WriteEnabled=false
    Module.Param3WriteEnabled=false
   )
  ]
  particleUpdateScript [
   &ParticleState(Kill Particles When Lifetime Has Elapsed=false, Let Infinitely Lived Particles Die When Emitter Deactivates=true)
   /SetVariables(Particles.MeshOrientation=$System.AutoRotate) {isSetParameters=true}
   /SetVariables(Particles.Position=&Add_Vector(A=<0 0 0>, B=$System.LogoOffset)) {isSetParameters=true}
  ]
  rendererClasses [&NiagaraMeshRendererProperties]
  renderers [
   {rendererClass=&NiagaraMeshRendererProperties, rendererIndex=0}
   {rendererClass=&NiagaraMeshRendererProperties, rendererIndex=1}
  ]
 }
 PresetFX_AberrationAndGhost {
  bEnabled=false
  emitterUpdateScript [
   &EmitterState(
    Inactive Response=::NewEnumerator1
    Life Cycle Mode=::NewEnumerator1
    Loop Behavior=::NewEnumerator1
    Loop Duration Mode=::NewEnumerator1
    UseLoopDelay=false
   )
   &SpawnBurst_Instantaneous(Use Spawn Probability=false, UseLoopCountLimit=false)
  ]
  particleSpawnScript [
   &InitializeParticle(
    Color Mode=::NewEnumerator0
    Lifetime=5
    Mesh Scale Mode=::NewEnumerator3
    Mesh Uniform Scale=0.7
    Sprite Size Mode=::NewEnumerator3
    Uniform Sprite Size=60
    UsePositionOffset=false
    Write Material Random=true
   )
   &DynamicMaterialParameters(
    Index 0 Param 1=$System.Random
    Module.Param0WriteEnabled=true
    Module.Param1WriteEnabled=false
    Module.Param2WriteEnabled=false
    Module.Param3WriteEnabled=false
   )
   /SetVariables(Particles.SpriteAlignment=<0 0 1>, Particles.SpriteFacing=<0 1 0>) {isSetParameters=true, module=SetVariables_#5}
  ]
  particleUpdateScript [
   &ParticleState(Kill Particles When Lifetime Has Elapsed=false, Let Infinitely Lived Particles Die When Emitter Deactivates=true)
   /SetVariables(
    Particles.SpriteAlignment=<0 0 1>
    Particles.SpriteFacing=&RotateVectorByQuaternion(Vector To Rotate=<0 1 0>, Quaternion=$System.AutoRotate)
   ) {isSetParameters=true}
  ]
  rendererClasses [&NiagaraSpriteRendererProperties]
  renderers [
   {rendererClass=&NiagaraSpriteRendererProperties}
   {rendererClass=&NiagaraSpriteRendererProperties, rendererIndex=1}
  ]
 }
 PresetFX_Glow {
  emitterUpdateScript [
   &EmitterState(
    Inactive Response=::NewEnumerator1
    Life Cycle Mode=::NewEnumerator1
    Loop Behavior=::NewEnumerator1
    Loop Duration Mode=::NewEnumerator1
    UseLoopDelay=false
   )
   &SpawnBurst_Instantaneous(Use Spawn Probability=false, UseLoopCountLimit=false)
  ]
  particleSpawnScript [&InitializeParticle(
    Color=<rgba 1 0.0389 0.1086 0.47>
    Sprite Size Mode=::NewEnumerator3
    Uniform Sprite Size=60.2099
    UsePositionOffset=false
    Write Material Random=true
   )]
  particleUpdateScript [
   &ParticleState(Kill Particles When Lifetime Has Elapsed=false)
   &ScaleColor(
    Scale RGB=&VectorFromFloat(Value=&FloatFromCurve(
      FloatCurve=DataInterface() {propertyValues {Curve {keys [<at 0 1>,<at 0.1 0.625>,<at 0.2 1.125>,<at 0.4 0.375>,<at 0.5 1.375>,<at 0.6 0.75>,<at 0.7 1.625>,<at 0.8 0.625>,<at 0.9 1.25>,<at 1 0.875>]}}}
      CurveIndex=&Multiply_Float(A=$System.LoopedAge, B=0.2)
     ))
    ScaleA=true
    ScaleRGB=true
   )
   /SetVariables(Particles.Position=&Add_Vector(A=<0 3 0>, B=$System.LogoOffset)) {isSetParameters=true, module=SetVariables_#4}
   &CameraOffset(Camera Offset Amount=10)
  ]
  rendererClasses [&NiagaraSpriteRendererProperties]
  renderers [{rendererClass=&NiagaraSpriteRendererProperties}]
 }
 PresetFX_RainbowLine {
  bEnabled=false
  emitterUpdateScript [
   &EmitterState(
    Inactive Response=::NewEnumerator1
    Life Cycle Mode=::NewEnumerator1
    Loop Behavior=::NewEnumerator1
    Loop Duration Mode=::NewEnumerator1
    UseLoopDelay=false
   )
   &SpawnBurst_Instantaneous(Use Spawn Probability=false, UseLoopCountLimit=false)
   /SetVariables(Emitter.RainbowLine_Flat:NiagaraFloat=&MakeCustomFloatFromBool(Bool=&SetBoolByFloatComparison(A=$User.IsFrom3D, B=0), True Float=0, False Float=-1)) {isSetParameters=true}
   /SetVariables(Emitter.RainbowLine_Strength:NiagaraFloat=&MakeCustomFloatFromBool(Bool=&SetBoolByFloatComparison(A=$User.IsFrom3D, B=0), True Float=4, False Float=20)) {isSetParameters=true}
   /SetVariables(Emitter.RainbowLine_Power=0.1) {isSetParameters=true}
   /SetVariables(Emitter.RainbowLine_Density=5) {isSetParameters=true}
   /SetVariables(Emitter.RainbowLine_FlowSpeed=0.15) {isSetParameters=true}
  ]
  particleSpawnScript [
   &InitializeParticle(
    Color Mode=::NewEnumerator0
    Lifetime=5
    Mesh Scale Mode=::NewEnumerator3
    Mesh Uniform Scale=0.7
    Sprite Size Mode=::NewEnumerator3
    Uniform Sprite Size=70
    UsePositionOffset=false
    Write Material Random=true
   )
   &DynamicMaterialParameters(
    Index 0 Param 1=$System.Random
    Module.Param0WriteEnabled=true
    Module.Param1WriteEnabled=false
    Module.Param2WriteEnabled=false
    Module.Param3WriteEnabled=false
   )
   /SetVariables(Particles.RainbowLineSize:Vector2f=&Multiply_Vector2D_ByFloat(Float=0.5, Vector2D=$Particles.SpriteSize)) {isSetParameters=true}
   /SetVariables(Particles.SpriteAlignment=<0 0 1>, Particles.SpriteFacing=<0 1 0>) {isSetParameters=true}
  ]
  particleUpdateScript [
   &ParticleState(Kill Particles When Lifetime Has Elapsed=false, Let Infinitely Lived Particles Die When Emitter Deactivates=true)
   /SetVariables(
    Particles.SpriteAlignment=<0 0 1>
    Particles.SpriteFacing=&RotateVectorByQuaternion(Vector To Rotate=<0 1 0>, Quaternion=$System.AutoRotate)
   ) {isSetParameters=true, module=SetVariables_#5}
  ]
  rendererClasses [&NiagaraSpriteRendererProperties]
  renderers [{rendererClass=&NiagaraSpriteRendererProperties}]
 }
}
summary {
 systemName=NS_EerieGhost_Backpack_Idle
 emitters [
  {emitterName=Core, rendererClasses [&NiagaraMeshRendererProperties]}
  {bEnabled=false, emitterName=PresetFX_AberrationAndGhost, rendererClasses [&NiagaraSpriteRendererProperties]}
  {bEnabled=false, emitterName=PresetFX_RainbowLine, rendererClasses [&NiagaraSpriteRendererProperties]}
  {emitterName=PresetFX_Glow, rendererClasses [&NiagaraSpriteRendererProperties]}
  {emitterName=BackingGlow, rendererClasses [&NiagaraSpriteRendererProperties]}
 ]
 userVariables [
  User.Backpack_DiffuseTex {
   type=Texture
   description=""
   defaultValue {
    struct=NiagaraExt_VariableValue_Object
    value {object=/BRCosmetics/Accessories/FORT_Backpacks/Backpack_M_MED_EerieGhost/FX/T_FX_EerieGhost_HologramSrc, objectClass=/Script/Engine.Texture}
   }
  }
  User.Backpack_Mat {
   type=MaterialInterface
   description=""
   defaultValue {
    struct=NiagaraExt_VariableValue_Object
    value {
     object=/BRCosmetics/Accessories/FORT_Backpacks/Backpack_M_MED_EerieGhost/FX/Material/MI_3DHB_EerieGhost
     objectClass=/Script/Engine.MaterialInterface
    }
   }
  }
  User.Backpack_NormalTex {
   type=Texture
   description=""
   defaultValue {
    struct=NiagaraExt_VariableValue_Object
    value {object=/BRCosmetics/Accessories/FORT_Backpacks/Backpack_M_MED_EerieGhost/FX/Texture/T__Normal, objectClass=&Texture}
   }
  }
  User.Backpack_WPOTex {
   type=Texture
   description=""
   defaultValue {
    struct=NiagaraExt_VariableValue_Object
    value {object=/BRCosmetics/Accessories/FORT_Backpacks/Backpack_M_MED_EerieGhost/FX/Texture/T__WPO, objectClass=&Texture}
   }
  }
  User.IsFrom3D {type=NiagaraFloat, defaultValue=0, description=""}
 ]
}
