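# Translucent unlit two-sided material for the Eerie Ghost back bling's hologram logo, usable on Niagara mesh particles.
# 3DHB base emissive gets noise-driven chromatic aberration via a quality switch, outer glow color and time-of-day tint.
# 3DHB HologramFX handles scanline and glitch effects, with dynamic parameters supplying random time offset and mesh scale.
# Packed mask channels provide separate alpha, scalable glow and a smoothstep ghost-face emissive mask; vertex red scales opacity.
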Material /BRCosmetics/Accessories/FORT_Backpacks/Backpack_M_MED_EerieGhost/FX/Material/M_BackpackEerieGhost_Logo
graph {
 properties {
  bEnableResponsiveAA=true
  bUsedWithNiagaraMeshParticles=true
  blendMode=BLEND_Translucent
  opacityMaskClipValue=0.333
  pixelDepthOffsetMode=PDOM_Legacy
  shadingModel=MSM_Unlit
  twoSided=true
  usedShadingModels=Unlit
 }
 nodes [
  n1 = DynamicParameter() {paramNames [Random,MeshScale,Param3,Param4]}
  n2 = MaterialFunctionCall() {materialFunction=&BoundingBoxBased_0-1_UVW}
  n3 = OneMinus(None=ComponentMask(None=VertexInterpolator(VS=@n2.RGB).PS) {b=true, r=true})
  n4 = MaterialFunctionCall!(
   Time=Add(A=Time() {bOverride_Period=true, period=6000}, B=@n1.Random)
   MeshScaleIn=@n1.MeshScale
   UVForScanline=@n3
   CustomFlicker
   UseCustomFlicker
   UVForStroke=@n3
   UVForGlitch=@n3
   BaseUVIn
   EmissiveIn=NamedRerouteUsage()
   OpacityIn=NamedRerouteUsage()
   WPOIn=NamedRerouteUsage()
  ) {materialFunction=/Game/Tools/3DHB/Shader/Base/MF_3DHB_HologramFX}
  n5 = ScalarParameter() {parameterName=GhostFace_SStep_Pos}
  n6 = TextureSample() {samplerType=SAMPLERTYPE_Masks, texture=/BRCosmetics/Accessories/FORT_Backpacks/Backpack_M_MED_EerieGhost/FX/T_FX_EerieGhost_Hologram_Masks}
  n7 = SmoothStep(Min=@n5, Max=Add(A=@n5, B=ScalarParameter() {parameterName=GhostFace_SStep_W}) {constB=0.1}, Value=OneMinus(None=@n6.B))
  MaterialFunctionCall(Emissive=Multiply(A=@n4.EmissiveOut, B=Clamp(None=@n7) {minDefault=0.25})) {materialFunction=&MF_FX_EmissiveAdjustments}
  n8 = Multiply(A=Multiply(A=@n4.UseForFX_GlitchToOpacityOut, B=@n4.UseForFX_GlitchToOpacityOut), B=NamedRerouteUsage())
  n9 = NamedRerouteUsage()
  Multiply(A=Saturate(None=Add(A=@n8, B=@n9)), B=VertexColor().R)
  n10 = MaterialFunctionCall!(MeshScaleIn=@n1.MeshScale) {materialFunction=/Game/Tools/3DHB/Shader/Base/MF_3DHB_Base}
  NamedRerouteDeclaration(None=ComponentMask(None=@n10.OriginDiffuse) {b=true, g=true, r=true}) {name=Base_Emissive, nodeColor=<rgba 0.5703 1 0 1>}
  NamedRerouteDeclaration(None=ComponentMask(None=@n10.OriginDiffuse) {a=true}) {name=Base_Opacity, nodeColor=<rgba 0.5703 1 0 1>}
  NamedRerouteDeclaration(None=@n10.OriginWPO) {name=Base_WPO, nodeColor=<rgba 0.5703 1 0 1>}
  n11 = VectorParameter() {parameterName=OuterGlowBaseColor, defaultValue=<rgba 0 0 0 1>}
  n12 = NamedRerouteUsage()
  n13 = TextureCoordinate()
  n14 = Multiply(A=TextureSample(UVs=Time()) {samplerType=SAMPLERTYPE_LinearColor, texture=/Game/Effects/Fort_Effects/Textures/Tile/LinerNoise64x64}.R) {constB=2}
  n15 = AppendVector(A=Multiply(A=Subtract(A=@n14), B=ScalarParameter() {parameterName=AberrationOffset}), B=Constant())
  n16 = Add(A=@n13, B=@n15)
  n17 = TextureObjectParameter() {group=Base, parameterName=Base_DiffuseTexture}
  n18 = Subtract(A=@n13, B=@n15)
  n19 = TextureObjectParameter() {group=Base, parameterName=Base_DiffuseTexture}
  n20 = AppendVector(A=AppendVector(A=TextureSample(UVs=@n16, Tex=@n17).R, B=TextureSample(UVs=@n18, Tex=@n19).G), B=ComponentMask(None=@n12) {b=true})
  n21 = QualitySwitch(Default=@n12, Low=NamedRerouteUsage(), High=@n20, Medium=@n20, Epic=@n20)
  n22 = NamedRerouteUsage()
  n23 = MaterialFunctionCall(
   Color_Morning=VectorParameter() {group=TimeOfDay, parameterName=ColorTint_Morning, defaultValue=<rgba 1 1 1 1>}.RGB
   Color_Day=VectorParameter() {group=TimeOfDay, parameterName=ColorTint_Day, defaultValue=<rgba 1 1 1 1>}.RGB
   Color_Evening=VectorParameter() {group=TimeOfDay, parameterName=ColorTint_Evening, defaultValue=<rgba 1 1 1 1>}.RGB
   Color_Night=VectorParameter() {group=TimeOfDay, parameterName=ColorTint_Night, defaultValue=<rgba 1 1 1 1>}.RGB
  ) {materialFunction=&Color_Time_of_Day}
  NamedRerouteDeclaration(None=Multiply(A=LinearInterpolate(A=@n11.RGB, B=@n21, Alpha=@n22), B=VertexInterpolator(VS=@n23.Result).PS)) {name=Post_Emissive, nodeColor=<rgba 0.5703 1 0 1>}
  NamedRerouteDeclaration(None=NamedRerouteUsage()) {name=Post_Opacity, nodeColor=<rgba 0.5703 1 0 1>}
  NamedRerouteDeclaration(None=NamedRerouteUsage()) {name=Post_WPO, nodeColor=<rgba 0.5703 1 0 1>}
  n24 = ScalarParameter() {parameterName=GlowScale}
  n25 = TextureSample(UVs=VertexInterpolator(VS=MaterialFunctionCall(Texture Scale=@n24) {materialFunction=&ScaleUVsByCenter}.UVs).PS) {samplerType=SAMPLERTYPE_Masks, texture=&T_FX_EerieGhost_Hologram_Masks}
  NamedRerouteDeclaration(None=Multiply(A=@n25.G, B=@n11.A) {constB=0.05}) {name=OuterGlow, nodeColor=<rgba 0.9453 0 1 1>}
  n26 = TextureSample() {samplerType=SAMPLERTYPE_Masks, texture=&T_FX_EerieGhost_Hologram_Masks}
  NamedRerouteDeclaration(None=@n26.R) {name=SeperateTexAlpha, nodeColor=<rgba 1 0.1172 0 1>}
 ]
}
