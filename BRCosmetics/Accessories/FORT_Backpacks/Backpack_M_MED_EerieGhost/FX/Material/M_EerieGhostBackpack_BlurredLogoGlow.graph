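# Additive unlit Niagara sprite material for the blurred logo glow effect on the Eerie Ghost backpack.
# A panned vertical coordinate feeds a repeating triangle-wave scanline pattern with density, thickness and sharpness controls.
# That pattern, scaled by an offset parameter, distorts the UVs sampling the green channel of the hologram mask texture.
# The mask multiplies particle alpha, while particle color passes through the shared emissive adjustments function.
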Material /BRCosmetics/Accessories/FORT_Backpacks/Backpack_M_MED_EerieGhost/FX/Material/M_EerieGhostBackpack_BlurredLogoGlow
graph {
 properties {bUsedWithNiagaraSprites=true, blendMode=BLEND_Additive, pixelDepthOffsetMode=PDOM_Legacy, shadingModel=MSM_Unlit, usedShadingModels=Unlit}
 nodes [
  n1 = ParticleColor()
  MaterialFunctionCall(Emissive=@n1.RGB) {materialFunction=&MF_FX_EmissiveAdjustments}
  n2 = TextureCoordinate()
  n3 = Constant()
  n4 = ScalarParameter() {defaultValue=0.01, group=Base_HoloFX_Scanline, parameterName=Scanline_FlowSpeed}
  n5 = Add(A=Panner(Coordinate=@n2) {speedY=0.1}, B=Multiply(A=AppendVector(A=@n3, B=@n4)))
  n6 = Frac(None=Multiply(
    A=MaterialFunctionCall(Float2=@n5) {materialFunction=&BreakOutFloat2Components}.G
    B=ScalarParameter() {defaultValue=250, group=Base_HoloFX_Scanline, parameterName=Scanline_Density}
   ))
  n7 = Multiply(A=Min(A=OneMinus(None=@n6), B=@n6)) {constB=2}
  n8 = ScalarParameter() {group=Base_HoloFX_Scanline, parameterName=Scanline_ThicknessOffset}
  n9 = Add(A=@n8)
  n10 = ScalarParameter() {group=Base_HoloFX_Scanline, parameterName=Scanline_Sharpness}
  n11 = Multiply(A=Multiply(A=Add(A=@n7, B=@n8), B=Divide(B=@n9)), B=Clamp(None=Add(A=@n10)) {clampMode=CMODE_ClampMin, minDefault=1})
  n12 = Subtract(A=Multiply(A=Saturate(None=@n11)) {constB=2})
  n13 = ScalarParameter() {parameterName=offset_amnt}
  n14 = TextureSample(UVs=Add(A=@n2, B=Multiply(A=@n12, B=@n13))) {samplerType=SAMPLERTYPE_Masks, texture=/BRCosmetics/Accessories/FORT_Backpacks/Backpack_M_MED_EerieGhost/FX/T_FX_EerieGhost_Hologram_Masks}
  Multiply(A=@n1.A, B=@n14.G)
 ]
}
